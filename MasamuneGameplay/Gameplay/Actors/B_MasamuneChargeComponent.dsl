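; Masamune charge component that accumulates a clamped charge value on a looping timer.
; Timer ticks add an amount scaled by a settable multiplier, with tick rate, per-tick gain and maximum read from scalable floats.
; Dispatchers report start, update, max and end; reaching max fires once and pauses the timer.
; Stopping can reset charge and multiplier, debug prints show the value, and passive recharge functions are unfinished stubs.
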(event EventBeginPlay)

(event Custom|StartCharging
  (bind _returnvalue (Utilities|Time|SetTimerbyEvent (delegate Custom|ChargeTick) (ScalableFloat|GetValueatLevel (Variables|Default|GetHF_TickRate)) true))
  (Variables|Default|SetChargeTimer _returnvalue))

(event Custom|ChargeTick
  (CallFunction|AddCharge (ScalableFloat|GetValueatLevel (Variables|Default|GetHF_ChargePerTick))))

(event Custom|StopCharging
  (params (Reset? :type Boolean) (AutoCancel? :type Boolean))
  (Utilities|Time|ClearandInvalidateTimerbyHandle (Variables|Default|GetChargeTimer))
  (Default|CallOnChargeEnded AutoCancel?)
  (if Reset?
    (CallFunction|ResetCharge)))

(event EventEndPlay
  (params (EndPlayReason :type EEndPlayReason%20Enum :default "Destroyed"))
  (CallFunction|StopCharging false))

(event Custom|AddCharge
  (params (Amount :type Float%20%28double-precision%29))
  (bind _output_get (Variables|Default|SetCharge (Math|Float|Clamp(Float) (+ (Variables|Default|GetCharge) (* Amount (Variables|Default|GetChargeMultiplier))) 0.0 (ScalableFloat|GetValueatLevel (Variables|Default|GetHF_MaxCharge)))))
  (Default|CallOnChargeUpdated _output_get)
  (Debugging|PrintFloat _output_get "T Charge = " 2.0 (Math|Color|SelectColor "(R=0.210702,G=0.843750,B=0.105469,A=1.000000)" "(R=1.000000,G=0.301597,B=0.203846,A=1.000000)" (> (Variables|Default|GetChargeMultiplier) 1.0))
    (:Then
      (if (CallFunction|IsMaxed?)
        (Utilities|FlowControl|DoOnce false
          (:Completed
            (Default|CallOnChargeMaxed)
            (Utilities|Time|PauseTimerbyHandle (Variables|Default|GetChargeTimer))))))))

(event Custom|ResetMaxCharge
  (Utilities|FlowControl|DoOnce false
    (:Completed
      (Default|CallOnChargeMaxed)
      (Utilities|Time|PauseTimerbyHandle (Variables|Default|GetChargeTimer)))))

(fn ApplyMultiplier
  (params (ChargeMult :type Float%20%28double-precision%29))
  (Variables|Default|SetChargeMultiplier ChargeMult))

(fn ApplyPassiveRecharge)

(fn EndPassiveRecharge)

(fn GetCharge)

(fn IsMaxed?
  (return (>= (Variables|Default|GetCharge) (ScalableFloat|GetValueatLevel (Variables|Default|GetHF_MaxCharge)))))

(fn ResetCharge
  (Variables|Default|SetChargeMultiplier 1.0)
  (bind _output_get (Variables|Default|SetCharge 0.0))
  (Default|CallOnChargeUpdated _output_get)
  (Utilities|Time|UnpauseTimerbyHandle (Variables|Default|GetChargeTimer))
  (CallFunction|ResetMaxCharge)
  (Default|CallOnChargeStarted))

(fn SetChargingPaused
  (params (Paused? :type Boolean))
  (bind _chargetimer (Variables|Default|GetChargeTimer))
  (if (CallFunction|IsMaxed?)
    _
    (elif Paused?
      (Utilities|Time|PauseTimerbyHandle _chargetimer)
      (else
        (Utilities|Time|UnpauseTimerbyHandle _chargetimer)))))

(fn StartPassiveRecharge
  (Utilities|Time|SetTimerbyEvent 0))
