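; Fire Petal Masamune weapon Blueprint that adds a custom tactical sprint MME to the player while equipped.
; Jump events during tactical sprint grant fall-damage immunity until landing, and an air-dash scalable float sets HUD action visibility at BeginPlay.
; Async tag queries watch sprint tags, and CheckForActionDisable toggles a loose tag cosmetically to disable the jump UI outside sprint.
; UpdateChargeTarget tracks the hit actor with a local lock-on GameplayCue.
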(event Custom|HandleFallImmunity
  (params (Equipped? :type Boolean))
  ;[0.7899 0.1002 0.2833] BeginPlay
  ;;[0.7899 0.1002 0.2833] Apply FDI when jumping
  (if Equipped?
    (bind _asynctaskproxy (Ability|Async|WaitGameplayEventToActor (Variables|Default|GetPlayerPawn) "(TagName=\"Event.Jump.Pressed\")")
      (:then
        (Variables|Default|SetWaitJumpTask _asynctaskproxy))
      (:EventReceived
        (CallFunction|TryApplyFallImmunity)))
    (else
      (Ability|Async|EndAction (Variables|Default|GetWaitJumpTask)))))

(event EventBeginPlay
  ;[0.7899 0.1002 0.2833] BeginPlay
  (Utilities|FlowControl|Sequence
    (:then_0
      (Variables|Default|SetPlayerPawn (Utilities|Casting|CastToFortPawn (Game|GetInstigator))))
    (:then_1
      (CallFunction|CheckJumpEnabled))))

(event Custom|TryApplyFallImmunity
  (bind _playerpawn (Variables|Default|GetPlayerPawn))
  ;[0.7899 0.1002 0.2833] BeginPlay
  ;;[0.7899 0.1002 0.2833] Apply FDI when jumping
  (if (GameplayTags|HasMatchingGameplayTag _playerpawn "(TagName=\"Gameplay.Action.Player.TacticalSprint\")")
    (Creative|GameplayEffects|SimpleApplyGameplayEffectToActor "/Game/Athena/Items/Parents/FallImmunity/GE_Athena_TillLandFallDamageImmunity_NoFX_Grant_Parent.GE_Athena_TillLandFallDamageImmunity_NoFX_Grant_Parent_C" _playerpawn)))

(event FortWeapon|EventOnEquippedAndAttached
  ;[0.7899 0.1002 0.2833] Sprint MME
  (CallFunction|ToggleMasamuneSprintMME true)
  (CallFunction|StartListeningforSprintTags)
  (CallFunction|HandleFallImmunity true))

(event Custom|StartListeningForSprintTags
  (bind _tacsprinttags (Variables|Default|GetTacSprintTags))
  (bind _returnvalue (GameplayTags|MakeGameplayTagQueryMatchAnyTags _tacsprinttags))
  (bind _tagquerytasks (Variables|Default|GetTagQueryTasks))
  (bind _playerpawn (Variables|Default|GetPlayerPawn))
  ;[0.7899 0.1002 0.2833] Listen for tags to know when to disable the jump UI
  (bind _asynctaskproxy (Ability|Tasks|WaitGameplayTagQueryOnActor _playerpawn _returnvalue)
    (:then
      (Utilities|Array|AddUnique _tagquerytasks _asynctaskproxy)
      (bind _asynctaskproxy_2 (Ability|Tasks|WaitGameplayTagQueryOnActor _playerpawn _returnvalue "WhenFalse")
        (:then
          (Utilities|Array|AddUnique _tagquerytasks _asynctaskproxy_2))
        (:Triggered
          _)))
    (:Triggered
      _))
  (CallFunction|CheckforActionDisable))

(event FortWeapon|EventOnUnEquippedAndDetached
  ;[0.7899 0.1002 0.2833] Sprint MME
  (CallFunction|ToggleMasamuneSprintMME false)
  (CallFunction|StopListeningforSprintTags)
  (CallFunction|HandleFallImmunity false))

(event Custom|StopListeningForSprintTags
  ;[0.7899 0.1002 0.2833] Listen for tags to know when to disable the jump UI
  (for _array_element (Variables|Default|GetTagQueryTasks)
    (Ability|Async|EndAction _array_element)))

(fn CheckForActionDisable
  (bind _tag (LocalVariables|GetTag))
  (bind _returnvalue (GameplayTags|MakeGameplayTagContainerfromTag _tag))
  (bind _returnvalue_2 (GameplayTags|HasMatchingGameplayTag (Variables|Default|GetPlayerPawn) _tag))
  (bind _playerpawn (Variables|Default|GetPlayerPawn))
  (Networking|CanExecuteCosmeticEvents
    (:True
      (if (GameplayTags|HasAnyMatchingGameplayTags _playerpawn (Variables|Default|GetTacSprintTags))
        (if _returnvalue_2
          (Ability|GameplayEffect|RemoveLooseGameplayTags _playerpawn _returnvalue))
        (elif _returnvalue_2
          _
          (else
            (Ability|GameplayEffect|AddLooseGameplayTags _playerpawn _returnvalue))))
      (return))))

(fn CheckJumpEnabled
  (WeaponHUDData|SetHudKeyActionVisibility 4 (ScalableFloat|GetValueasBool (Variables|Default|GetHF_AirDashEnable))))

(fn CreateDynamicMaterials)

(fn OnMoveModeChanged
  (params (Character :type Character%20Object%20Reference) (PrevMovementMode :type EMovementMode%20Enum :default "MOVE_None") (PreviousCustomMode :type Byte)))

(fn ToggleMasamuneSprintMME
  (params (Enable :type Boolean))
  (bind _playerpawn (Variables|Default|GetPlayerPawn))
  (bind _returnvalue (Pawn|GetController _playerpawn))
  (bind _customsprintmme (Variables|Default|GetCustomSprintMME))
  (Variables|Default|GetPawnSprintComponent
    (:then
      _)
    (:else
      (Utilities|IsValid _playerpawn
        (:Is%20Valid
          (Utilities|Casting|CastToPlayerController _returnvalue)
          (Variables|Default|SetPawnSprintComponent (Actor|GetComponentbyClass _returnvalue "/TacticalSprintGame/Gameplay/TacticalSprintControllerComponent.TacticalSprintControllerComponent_C"))))))
  (if Enable
    (Sprint|AddCustomSprintMMEClass (Variables|Default|GetPawnSprintComponent) _customsprintmme)
    (else
      (Sprint|RemoveCustomSprintMMEClass (Variables|Default|GetPawnSprintComponent) _customsprintmme)
      (Variables|Default|SetPawnSprintComponent 0))))

(fn UpdateChargeTarget
  (params (FoundTarget? :type Boolean) (Hit :type Hit%20Result%20Structure))
  (bind (_bblockinghit _binitialoverlap _time _distance _location _impactpoint _normal _impactnormal _physmat _hitactor) (Collision|BreakHitResult Hit))
  (bind _chargetarget (Variables|Charge|GetChargeTarget))
  (bind _gc_lockon (Variables|Default|GetGC_LockOn))
  (if FoundTarget?
    (if (!= _hitactor (Variables|Charge|GetChargeTarget))
      (Variables|Charge|SetHasChargeTarget? true)
      (Utilities|FlowControl|Sequence
        (:then_0
          _)
        (:then_1
          (bind _output_get (Variables|Charge|SetChargeTarget _hitactor))
          (Projectile|AddGameplayCue%28Looping%29Local _output_get _gc_lockon (GameplayCue|MakeGameplayCueParametersfromHitResult Hit))
          (Debugging|PrintObject _output_get "target successful = " 2.0 "(R=1.000000,G=0.301597,B=0.203846,A=1.000000)"))))
    (else
      (Variables|Charge|SetHasChargeTarget? false)
      (Utilities|IsValid _chargetarget
        (:Is%20Valid
          (Utilities|FlowControl|Sequence
            (:then_0
              (Projectile|RemoveGameplayCue%28Looping%29Local _chargetarget _gc_lockon (Ability|GameplayCue|MakeGameplayCueParameters 0.0)))
            (:then_1
              _)))
        (:Is%20Not%20Valid
          _))
      (Variables|Charge|SetChargeTarget 0))))

(fn ConstructionScript
  (|CallParentFunction))
